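# Sound cue for the Grandma creature's tame vocalization in the Irwin content.
# A random node picks one of three equally weighted wave variations on each play.
# Playback uses predator vocalization attenuation, a tactical creature vocalization sound class and a boosted volume.
# Concurrency combines footstep local and global limits with a rule that allows one instance and blocks new ones.
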SoundCue /Irwin/Sounds/Grandma/Vox/Grandma_Tame_Cue
root=Random_0
nodes [
 {id=Random_0, children [WavePlayer_0,WavePlayer_1,WavePlayer_2], properties {weights [1,1,1]}}
 {id=WavePlayer_0, properties {soundWaveAssetPtr=&/Grandma_Tame_01}}
 {id=WavePlayer_1, properties {soundWaveAssetPtr=&/Grandma_Tame_02}}
 {id=WavePlayer_2, properties {soundWaveAssetPtr=&/Grandma_Tame_03}}
]
properties {
 attenuationSettings=/Irwin/Sounds/Settings/Predator_Vocalization_Att
 maxDistance=8300
 soundClassObject=&Fort_Creatures_Vocalization_Tactical
 volumeMultiplier=1.8
 concurrencySet [
  &Athena_Footstep_Local_Conc
  &Athena_Footstep_Global_Conc
  /Game/Sounds/Concurrency_Rules/LimitToOnePreventNew
 ]
}
