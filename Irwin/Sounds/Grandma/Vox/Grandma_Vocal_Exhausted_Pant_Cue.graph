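# Sound cue that plays Grandma's exhausted panting vocal, classed as a tactical creature vocalization with close-range prey attenuation.
# A quality switch sends its first two levels to a six-clip random pool and its last two to a separate four-clip pool.
# Only the four-clip random node is excluded from branch culling.
# A root modulator lowers the pitch and varies it slightly, along with volume, on each play.
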SoundCue /Irwin/Sounds/Grandma/Vox/Grandma_Vocal_Exhausted_Pant_Cue
root=Modulator_0
nodes [
 {id=Modulator_0, children [QualityLevel_0], properties {pitchMax=0.85, pitchMin=0.8, volumeMin=0.75}}
 {id=QualityLevel_0, children [Random_0,Random_0,Random_1,Random_1]}
 {id=Random_0, children [WavePlayer_0,WavePlayer_1,WavePlayer_2,WavePlayer_3,WavePlayer_4,WavePlayer_5], properties {weights [1,1,1,1,1,1]}}
 {
  id=Random_1
  children [WavePlayer_6,WavePlayer_7,WavePlayer_8,WavePlayer_9]
  properties {bShouldExcludeFromBranchCulling=true, weights [1,1,1,1]}
 }
 {id=WavePlayer_0, properties {soundWaveAssetPtr=&/Grandma_Vocal_Exhausted_Pant_01}}
 {id=WavePlayer_1, properties {soundWaveAssetPtr=&/Grandma_Vocal_Exhausted_Pant_02}}
 {id=WavePlayer_2, properties {soundWaveAssetPtr=&/Grandma_Vocal_Exhausted_Pant_03}}
 {id=WavePlayer_3, properties {soundWaveAssetPtr=&/Grandma_Vocal_Exhausted_Pant_04}}
 {id=WavePlayer_4, properties {soundWaveAssetPtr=&/Grandma_Vocal_Exhausted_Pant_05}}
 {id=WavePlayer_5, properties {soundWaveAssetPtr=&/Grandma_Vocal_Exhausted_Pant_06}}
 {id=WavePlayer_6, properties {soundWaveAssetPtr=&Grandma_Vocal_Exhausted_Pant_01}}
 {id=WavePlayer_7, properties {soundWaveAssetPtr=&Grandma_Vocal_Exhausted_Pant_02}}
 {id=WavePlayer_8, properties {soundWaveAssetPtr=&Grandma_Vocal_Exhausted_Pant_03}}
 {id=WavePlayer_9, properties {soundWaveAssetPtr=&Grandma_Vocal_Exhausted_Pant_04}}
]
properties {
 attenuationSettings=/Irwin/Sounds/Settings/Prey_Vocalization_Close_Att
 maxDistance=2150
 soundClassObject=&Fort_Creatures_Vocalization_Tactical
 volumeMultiplier=0.5
}
