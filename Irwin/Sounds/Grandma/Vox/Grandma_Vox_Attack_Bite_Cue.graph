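# Sound cue for Grandma's vocal bite attack, using a predator vocalization attenuation and a creature vocalization sound class.
# A quality-level node sends its first two slots to a random pick of eleven bite takes.
# Its last two slots use a Modulator over a random pick of three of those takes, excluded from branch culling.
# The cue boosts volume and shares local and global footstep concurrency sets.
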SoundCue /Irwin/Sounds/Grandma/Vox/Grandma_Vox_Attack_Bite_Cue
root=QualityLevel_0
nodes [
 {id=Modulator_0, children [Random_1]}
 {id=QualityLevel_0, children [Random_0,Random_0,Modulator_0,Modulator_0]}
 {
  id=Random_0
  children [
   WavePlayer_0
   WavePlayer_1
   WavePlayer_2
   WavePlayer_3
   WavePlayer_4
   WavePlayer_5
   WavePlayer_6
   WavePlayer_7
   WavePlayer_8
   WavePlayer_9
   WavePlayer_10
  ]
  properties {weights [1,1,1,1,1,1,1,1,1,1,1]}
 }
 {id=Random_1, children [WavePlayer_11, WavePlayer_12, WavePlayer_13], properties {bShouldExcludeFromBranchCulling=true, weights [1,1,1]}}
 {id=WavePlayer_0, properties {soundWaveAssetPtr=&/Grandma_Vox_Attack_Bite_09}}
 {id=WavePlayer_1, properties {soundWaveAssetPtr=&/Grandma_Vox_Attack_Bite_10}}
 {id=WavePlayer_10, properties {soundWaveAssetPtr=&/Grandma_Vox_Attack_Bite_08}}
 {id=WavePlayer_11, properties {soundWaveAssetPtr=&/Grandma_Vox_Attack_Bite_01}}
 {id=WavePlayer_12, properties {soundWaveAssetPtr=&/Grandma_Vox_Attack_Bite_06}}
 {id=WavePlayer_13, properties {soundWaveAssetPtr=&Grandma_Vox_Attack_Bite_08}}
 {id=WavePlayer_2, properties {soundWaveAssetPtr=&/Grandma_Vox_Attack_Bite_11}}
 {id=WavePlayer_3, properties {soundWaveAssetPtr=&Grandma_Vox_Attack_Bite_01}}
 {id=WavePlayer_4, properties {soundWaveAssetPtr=&/Grandma_Vox_Attack_Bite_02}}
 {id=WavePlayer_5, properties {soundWaveAssetPtr=&/Grandma_Vox_Attack_Bite_03}}
 {id=WavePlayer_6, properties {soundWaveAssetPtr=&/Grandma_Vox_Attack_Bite_04}}
 {id=WavePlayer_7, properties {soundWaveAssetPtr=&/Grandma_Vox_Attack_Bite_05}}
 {id=WavePlayer_8, properties {soundWaveAssetPtr=&Grandma_Vox_Attack_Bite_06}}
 {id=WavePlayer_9, properties {soundWaveAssetPtr=&/Grandma_Vox_Attack_Bite_07}}
]
properties {
 attenuationSettings=/Irwin/Sounds/Settings/Predator_Vocalization_Att
 maxDistance=8300
 soundClassObject=&Fort_Creatures_Vocalization_Tactical
 volumeMultiplier=2.5
 concurrencySet [&Athena_Footstep_Local_Conc, &Athena_Footstep_Global_Conc]
}
