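# Sound cue for the Grandma creature's lunge attack vocalization, played with predator vocalization attenuation and a boosted volume.
# A quality-level node picks randomly among six lunge waves in its first two slots.
# Its last two slots use a smaller three-wave random pool through a modulator for pitch variation, compensating for fewer samples.
# Despite being a vocal, it shares the Athena local and global footstep concurrency limits.
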SoundCue /Irwin/Sounds/Grandma/Vox/Grandma_Vox_Attack_Lunge_Cue
root=QualityLevel_0
nodes [
 {id=Modulator_0, children [Random_1], properties {pitchMin=0.9}}
 {id=QualityLevel_0, children [Random_0,Random_0,Modulator_0,Modulator_0]}
 {id=Random_0, children [WavePlayer_0,WavePlayer_1,WavePlayer_2,WavePlayer_3,WavePlayer_4,WavePlayer_5], properties {weights [1,1,1,1,1,1]}}
 {id=Random_1, children [WavePlayer_6,WavePlayer_7,WavePlayer_8], properties {bShouldExcludeFromBranchCulling=true, weights [1,1,1]}}
 {id=WavePlayer_0, properties {soundWaveAssetPtr=&/Grandma_Vox_Attack_Lunge_06}}
 {id=WavePlayer_1, properties {soundWaveAssetPtr=&/Grandma_Vox_Attack_Lunge_01}}
 {id=WavePlayer_2, properties {soundWaveAssetPtr=&/Grandma_Vox_Attack_Lunge_02}}
 {id=WavePlayer_3, properties {soundWaveAssetPtr=&/Grandma_Vox_Attack_Lunge_03}}
 {id=WavePlayer_4, properties {soundWaveAssetPtr=&/Grandma_Vox_Attack_Lunge_04}}
 {id=WavePlayer_5, properties {soundWaveAssetPtr=&/Grandma_Vox_Attack_Lunge_05}}
 {id=WavePlayer_6, properties {soundWaveAssetPtr=&Grandma_Vox_Attack_Lunge_06}}
 {id=WavePlayer_7, properties {soundWaveAssetPtr=&Grandma_Vox_Attack_Lunge_04}}
 {id=WavePlayer_8, properties {soundWaveAssetPtr=&Grandma_Vox_Attack_Lunge_05}}
]
properties {
 attenuationSettings=/Irwin/Sounds/Settings/Predator_Vocalization_Att
 maxDistance=8300
 soundClassObject=&Fort_Creatures_Vocalization_Tactical
 volumeMultiplier=2.5
 concurrencySet [&Athena_Footstep_Local_Conc, &Athena_Footstep_Global_Conc]
}
