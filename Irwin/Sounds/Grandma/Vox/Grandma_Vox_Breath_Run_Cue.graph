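# Sound cue for the running breath vocalizations of Grandma, a rideable creature, shifting toward heavier breathing and tiring whines as FatigueLevel rises.
# IsLocal and IsRiddenByEnemy branches choose sound class, attenuation and volume, then share a randomized volume modulator.
# A quality switch picks larger or smaller wave pools, each crossfaded on FatigueLevel, with IsEnemy choosing aggressive or normal breaths when fatigue is low.
# Null random slots leave some triggers silent, and concurrency rules limit overlap.
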SoundCue /Irwin/Sounds/Grandma/Vox/Grandma_Vox_Breath_Run_Cue
root=Branch_4
nodes [
 {id=Attenuation_0, children [Modulator_2], properties {attenuationSettings=/Irwin/Sounds/Settings/Predator_Vocalization_Att}}
 {id=Attenuation_5, children [Modulator_2], properties {attenuationSettings=/Game/Sounds/Attenuation/Fort_Riding_Enemy_VO_Att}}
 {id=Attenuation_6, children [Modulator_2], properties {attenuationSettings=&Predator_Vocalization_Att}}
 {id=Attenuation_7, children [Modulator_2], properties {attenuationSettings=&Predator_Vocalization_Att}}
 {id=Branch_1, children [Random_2,Random_0,Random_0], properties {boolParameterName=IsEnemy}}
 {id=Branch_2, children [Random_8,Random_1,Random_1], properties {boolParameterName=IsEnemy}}
 {id=Branch_3, children [Modulator_5,Modulator_3,SoundClass_0], properties {boolParameterName=IsRiddenByEnemy}}
 {id=Branch_4, children [SoundClass_7,Branch_3,Branch_3], properties {boolParameterName=IsLocal}}
 {id=Modulator_2, children [QualityLevel_0], properties {volumeMax=1, volumeMin=0.5}}
 {id=Modulator_3, children [SoundClass_6], properties {pitchMax=1, pitchMin=1, volumeMax=0.1, volumeMin=0.1}}
 {id=Modulator_5, children [SoundClass_5], properties {pitchMax=1, pitchMin=1, volumeMax=0.7, volumeMin=0.7}}
 {
  id=ParamCrossFade_0
  children [Branch_2,Random_5,Random_9]
  properties {
   paramName=FatigueLevel
   crossFadeInput [
    {fadeInDistanceEnd=0, fadeInDistanceStart=0, fadeOutDistanceEnd=0.85, fadeOutDistanceStart=0.82, volume=1}
    {fadeInDistanceEnd=0.85, fadeInDistanceStart=0.82, fadeOutDistanceEnd=0.96, fadeOutDistanceStart=0.95, volume=1.2}
    {fadeInDistanceEnd=0.96, fadeInDistanceStart=0.95, fadeOutDistanceEnd=1, fadeOutDistanceStart=1, volume=1.2}
   ]
  }
 }
 {
  id=ParamCrossFade_1
  children [Branch_1,Random_3,Random_11]
  properties {
   paramName=FatigueLevel
   crossFadeInput [
    {fadeInDistanceEnd=0, fadeInDistanceStart=0, fadeOutDistanceEnd=0.85, fadeOutDistanceStart=0.82, volume=1.8}
    {fadeInDistanceEnd=0.85, fadeInDistanceStart=0.82, fadeOutDistanceEnd=0.96, fadeOutDistanceStart=0.95, volume=2.5}
    {fadeInDistanceEnd=0.96, fadeInDistanceStart=0.95, fadeOutDistanceEnd=1, fadeOutDistanceStart=1, volume=2.2}
   ]
  }
 }
 {id=QualityLevel_0, children [ParamCrossFade_1, ParamCrossFade_1, ParamCrossFade_0, ParamCrossFade_0]}
 {
  id=Random_0
  children [
   WavePlayer_0
   WavePlayer_1
   WavePlayer_2
   WavePlayer_3
   WavePlayer_4
   WavePlayer_5
   WavePlayer_6
   WavePlayer_7
   WavePlayer_8
   WavePlayer_9
   WavePlayer_10
   WavePlayer_11
   WavePlayer_12
   WavePlayer_13
   WavePlayer_14
   WavePlayer_15
   WavePlayer_16
   WavePlayer_17
   WavePlayer_18
   WavePlayer_19
   null
   null
   null
   null
   null
   null
   null
   null
   null
   null
   null
   null
  ]
  properties {bShouldExcludeFromBranchCulling=true, weights [1,1,1,1,1,1,1,1,1,1,1,1,1,1,1,1,1,1,1,1,1,1,1,1,1,1,1,1,1,1,1,1]}
 }
 {
  id=Random_1
  children [WavePlayer_20, WavePlayer_21, WavePlayer_22, WavePlayer_23, null, null, null, null]
  properties {bShouldExcludeFromBranchCulling=true, weights [1,1,1,1,1,1,1,1]}
 }
 {
  id=Random_11
  children [WavePlayer_95, WavePlayer_96, WavePlayer_97, WavePlayer_98, WavePlayer_99, WavePlayer_100, WavePlayer_101, null, null, null]
  properties {bShouldExcludeFromBranchCulling=true, weights [1,1,1,1,1,1,1,1,1,1]}
 }
 {
  id=Random_2
  children [
   WavePlayer_24
   WavePlayer_25
   WavePlayer_26
   WavePlayer_27
   WavePlayer_28
   WavePlayer_29
   WavePlayer_30
   WavePlayer_31
   WavePlayer_32
   WavePlayer_33
   WavePlayer_34
   WavePlayer_35
   WavePlayer_36
   WavePlayer_37
   WavePlayer_38
   WavePlayer_39
   WavePlayer_40
   WavePlayer_41
   null
   null
   null
   null
   null
   null
   null
   null
   null
   null
   null
   null
   null
  ]
  properties {bShouldExcludeFromBranchCulling=true, weights [1,1,1,1,1,1,1,1,1,1,1,1,1,1,1,1,1,1,1,1,1,1,1,1,1,1,1,1,1,1,1]}
 }
 {
  id=Random_3
  children [
   WavePlayer_42
   WavePlayer_43
   WavePlayer_44
   WavePlayer_45
   WavePlayer_46
   WavePlayer_47
   WavePlayer_48
   WavePlayer_49
   null
   null
   null
   null
   null
  ]
  properties {bShouldExcludeFromBranchCulling=true, weights [1,1,1,1,1,1,1,1,1,1,1,1,1]}
 }
 {
  id=Random_5
  children [WavePlayer_73, WavePlayer_76, WavePlayer_77, WavePlayer_78, WavePlayer_79, WavePlayer_80, null, null, null, null, null, null, null]
  properties {bShouldExcludeFromBranchCulling=true, weights [1,1,1,1,1,1,1,1,1,1,1,1,1]}
 }
 {
  id=Random_8
  children [WavePlayer_67, WavePlayer_69, WavePlayer_88, WavePlayer_89, null, null, null, null]
  properties {bShouldExcludeFromBranchCulling=true, weights [1,1,1,1,1,1,1,1]}
 }
 {
  id=Random_9
  children [WavePlayer_81, WavePlayer_82, WavePlayer_83, WavePlayer_85, WavePlayer_86, WavePlayer_87, null, null, null]
  properties {bShouldExcludeFromBranchCulling=true, weights [1,1,1,1,1,1,1,1,1]}
 }
 {
  id=SoundClass_0
  children [Attenuation_0]
  properties {soundClassOverride=&Fort_Creatures_Vocalization_Tactical}
 }
 {
  id=SoundClass_5
  children [Attenuation_5]
  properties {soundClassOverride=&Fort_Player_Foley_3P_Enemy_Athena}
 }
 {
  id=SoundClass_6
  children [Attenuation_6]
  properties {soundClassOverride=&Fort_Creatures_Vocalization_Tactical}
 }
 {
  id=SoundClass_7
  children [Attenuation_7]
  properties {soundClassOverride=&Fort_Creatures_Vocalization_Tactical}
 }
 {id=WavePlayer_0, properties {soundWaveAssetPtr=&/Grandma_Vox_Run_Breath_10}}
 {id=WavePlayer_1, properties {soundWaveAssetPtr=&/Grandma_Vox_Run_Breath_11}}
 {id=WavePlayer_10, properties {soundWaveAssetPtr=&/Grandma_Vox_Run_Breath_20}}
 {id=WavePlayer_100, properties {soundWaveAssetPtr=&/Grandma_Vox_Swim_Aggro_Breath_02}}
 {id=WavePlayer_101, properties {soundWaveAssetPtr=&/Grandma_Vox_Swim_Aggro_Breath_03}}
 {id=WavePlayer_11, properties {soundWaveAssetPtr=&/Grandma_Vox_Run_Breath_01}}
 {id=WavePlayer_12, properties {soundWaveAssetPtr=&/Grandma_Vox_Run_Breath_02}}
 {id=WavePlayer_13, properties {soundWaveAssetPtr=&/Grandma_Vox_Run_Breath_03}}
 {id=WavePlayer_14, properties {soundWaveAssetPtr=&/Grandma_Vox_Run_Breath_04}}
 {id=WavePlayer_15, properties {soundWaveAssetPtr=&/Grandma_Vox_Run_Breath_05}}
 {id=WavePlayer_16, properties {soundWaveAssetPtr=&/Grandma_Vox_Run_Breath_06}}
 {id=WavePlayer_17, properties {soundWaveAssetPtr=&/Grandma_Vox_Run_Breath_07}}
 {id=WavePlayer_18, properties {soundWaveAssetPtr=&/Grandma_Vox_Run_Breath_08}}
 {id=WavePlayer_19, properties {soundWaveAssetPtr=&/Grandma_Vox_Run_Breath_09}}
 {id=WavePlayer_2, properties {soundWaveAssetPtr=&/Grandma_Vox_Run_Breath_12}}
 {id=WavePlayer_20, properties {soundWaveAssetPtr=&Grandma_Vox_Run_Breath_01}}
 {id=WavePlayer_21, properties {soundWaveAssetPtr=&Grandma_Vox_Run_Breath_02}}
 {id=WavePlayer_22, properties {soundWaveAssetPtr=&Grandma_Vox_Run_Breath_03}}
 {id=WavePlayer_23, properties {soundWaveAssetPtr=&Grandma_Vox_Run_Breath_04}}
 {id=WavePlayer_24, properties {soundWaveAssetPtr=&/Grandma_Vox_Run_Aggro_Breath_18}}
 {id=WavePlayer_25, properties {soundWaveAssetPtr=&/Grandma_Vox_Run_Aggro_Breath_01}}
 {id=WavePlayer_26, properties {soundWaveAssetPtr=&/Grandma_Vox_Run_Aggro_Breath_02}}
 {id=WavePlayer_27, properties {soundWaveAssetPtr=&/Grandma_Vox_Run_Aggro_Breath_03}}
 {id=WavePlayer_28, properties {soundWaveAssetPtr=&/Grandma_Vox_Run_Aggro_Breath_04}}
 {id=WavePlayer_29, properties {soundWaveAssetPtr=&/Grandma_Vox_Run_Aggro_Breath_05}}
 {id=WavePlayer_3, properties {soundWaveAssetPtr=&/Grandma_Vox_Run_Breath_13}}
 {id=WavePlayer_30, properties {soundWaveAssetPtr=&/Grandma_Vox_Run_Aggro_Breath_06}}
 {id=WavePlayer_31, properties {soundWaveAssetPtr=&/Grandma_Vox_Run_Aggro_Breath_07}}
 {id=WavePlayer_32, properties {soundWaveAssetPtr=&/Grandma_Vox_Run_Aggro_Breath_08}}
 {id=WavePlayer_33, properties {soundWaveAssetPtr=&/Grandma_Vox_Run_Aggro_Breath_09}}
 {id=WavePlayer_34, properties {soundWaveAssetPtr=&/Grandma_Vox_Run_Aggro_Breath_10}}
 {id=WavePlayer_35, properties {soundWaveAssetPtr=&/Grandma_Vox_Run_Aggro_Breath_11}}
 {id=WavePlayer_36, properties {soundWaveAssetPtr=&/Grandma_Vox_Run_Aggro_Breath_12}}
 {id=WavePlayer_37, properties {soundWaveAssetPtr=&/Grandma_Vox_Run_Aggro_Breath_13}}
 {id=WavePlayer_38, properties {soundWaveAssetPtr=&/Grandma_Vox_Run_Aggro_Breath_14}}
 {id=WavePlayer_39, properties {soundWaveAssetPtr=&/Grandma_Vox_Run_Aggro_Breath_15}}
 {id=WavePlayer_4, properties {soundWaveAssetPtr=&/Grandma_Vox_Run_Breath_14}}
 {id=WavePlayer_40, properties {soundWaveAssetPtr=&/Grandma_Vox_Run_Aggro_Breath_16}}
 {id=WavePlayer_41, properties {soundWaveAssetPtr=&/Grandma_Vox_Run_Aggro_Breath_17}}
 {id=WavePlayer_42, properties {soundWaveAssetPtr=&/Grandma_Vocal_Heavy_Breathing_01}}
 {id=WavePlayer_43, properties {soundWaveAssetPtr=&/Grandma_Vocal_Heavy_Breathing_02}}
 {id=WavePlayer_44, properties {soundWaveAssetPtr=&/Grandma_Vocal_Heavy_Breathing_03}}
 {id=WavePlayer_45, properties {soundWaveAssetPtr=&/Grandma_Vocal_Heavy_Breathing_04}}
 {id=WavePlayer_46, properties {soundWaveAssetPtr=&/Grandma_Vocal_Heavy_Breathing_05}}
 {id=WavePlayer_47, properties {soundWaveAssetPtr=&/Grandma_Vocal_Heavy_Breathing_06}}
 {id=WavePlayer_48, properties {soundWaveAssetPtr=&/Grandma_Vocal_Heavy_Breathing_07}}
 {id=WavePlayer_49, properties {soundWaveAssetPtr=&/Grandma_Vocal_Heavy_Breathing_08}}
 {id=WavePlayer_5, properties {soundWaveAssetPtr=&/Grandma_Vox_Run_Breath_15}}
 {id=WavePlayer_6, properties {soundWaveAssetPtr=&/Grandma_Vox_Run_Breath_16}}
 {id=WavePlayer_67, properties {soundWaveAssetPtr=&Grandma_Vox_Run_Aggro_Breath_01}}
 {id=WavePlayer_69, properties {soundWaveAssetPtr=&Grandma_Vox_Run_Aggro_Breath_02}}
 {id=WavePlayer_7, properties {soundWaveAssetPtr=&/Grandma_Vox_Run_Breath_17}}
 {id=WavePlayer_73, properties {soundWaveAssetPtr=&/Grandma_Vox_Swim_Aggro_Breath_01}}
 {id=WavePlayer_76, properties {soundWaveAssetPtr=&Grandma_Vox_Swim_Aggro_Breath_02}}
 {id=WavePlayer_77, properties {soundWaveAssetPtr=&Grandma_Vox_Swim_Aggro_Breath_03}}
 {id=WavePlayer_78, properties {soundWaveAssetPtr=&/Grandma_Vox_Swim_Aggro_Breath_07}}
 {id=WavePlayer_79, properties {soundWaveAssetPtr=&/Grandma_Vox_Swim_Aggro_Breath_08}}
 {id=WavePlayer_8, properties {soundWaveAssetPtr=&/Grandma_Vox_Run_Breath_18}}
 {id=WavePlayer_80, properties {soundWaveAssetPtr=&/Grandma_Vox_Swim_Aggro_Breath_09}}
 {id=WavePlayer_81, properties {soundWaveAssetPtr=&/Grandma_Vocal_Sprint_Tiring_Whines_01}}
 {id=WavePlayer_82, properties {soundWaveAssetPtr=&/Grandma_Vocal_Sprint_Tiring_Whines_02}}
 {id=WavePlayer_83, properties {soundWaveAssetPtr=&/Grandma_Vocal_Sprint_Tiring_Whines_03}}
 {id=WavePlayer_85, properties {soundWaveAssetPtr=&Grandma_Vox_Swim_Aggro_Breath_01}}
 {id=WavePlayer_86, properties {soundWaveAssetPtr=&Grandma_Vox_Swim_Aggro_Breath_02}}
 {id=WavePlayer_87, properties {soundWaveAssetPtr=&Grandma_Vox_Swim_Aggro_Breath_03}}
 {id=WavePlayer_88, properties {soundWaveAssetPtr=&Grandma_Vox_Run_Aggro_Breath_03}}
 {id=WavePlayer_89, properties {soundWaveAssetPtr=&Grandma_Vox_Run_Aggro_Breath_04}}
 {id=WavePlayer_9, properties {soundWaveAssetPtr=&/Grandma_Vox_Run_Breath_19}}
 {id=WavePlayer_95, properties {soundWaveAssetPtr=&Grandma_Vocal_Sprint_Tiring_Whines_01}}
 {id=WavePlayer_96, properties {soundWaveAssetPtr=&Grandma_Vocal_Sprint_Tiring_Whines_02}}
 {id=WavePlayer_97, properties {soundWaveAssetPtr=&Grandma_Vocal_Sprint_Tiring_Whines_03}}
 {id=WavePlayer_98, properties {soundWaveAssetPtr=&/Grandma_Vocal_Sprint_Tiring_Whines_04}}
 {id=WavePlayer_99, properties {soundWaveAssetPtr=&Grandma_Vox_Swim_Aggro_Breath_01}}
]
properties {
 maxDistance=9000
 volumeMultiplier=1.3
 concurrencySet [
  &Athena_Footstep_Local_Conc
  &Athena_Footstep_Global_Conc
  /Game/Sounds/Concurrency_Rules/LimitToOnePreventNewQuick
 ]
}
