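# Sound cue for Grandma creature walking breath vocals, pitched down and limited by footstep and single-instance concurrency.
# IsLocal and IsRiddenByEnemy branches choose sound class, attenuation and volume before a shared random volume modulator.
# A quality-level switch picks between full and smaller wave pools, each crossfaded by FatigueLevel.
# Low fatigue branches on IsEnemy between aggro and walk breaths; higher fatigue brings heavy breathing, then tiring whines at full quality.
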SoundCue /Irwin/Sounds/Grandma/Vox/Grandma_Vox_Breath_Walk_Cue
root=Branch_4
nodes [
 {id=Attenuation_0, children [Modulator_3], properties {attenuationSettings=/Game/Sounds/Attenuation/Fort_Riding_Enemy_VO_Att}}
 {id=Attenuation_1, children [Modulator_3], properties {attenuationSettings=/Irwin/Sounds/Settings/Predator_Vocalization_Att}}
 {id=Attenuation_2, children [Modulator_3], properties {attenuationSettings=&Predator_Vocalization_Att}}
 {id=Attenuation_7, children [Modulator_3], properties {attenuationSettings=&Predator_Vocalization_Att}}
 {id=Branch_0, children [Modulator_5,Modulator_4,SoundClass_2], properties {boolParameterName=IsRiddenByEnemy}}
 {id=Branch_1, children [Random_2,Random_0,Random_0], properties {boolParameterName=IsEnemy}}
 {id=Branch_2, children [Random_5,Random_13,Random_13], properties {boolParameterName=IsEnemy}}
 {id=Branch_4, children [SoundClass_7,Branch_0,Branch_0], properties {boolParameterName=IsLocal}}
 {id=Modulator_3, children [QualityLevel_0], properties {volumeMax=1, volumeMin=0.5}}
 {id=Modulator_4, children [SoundClass_1], properties {pitchMax=1, pitchMin=1, volumeMax=0.1, volumeMin=0.1}}
 {id=Modulator_5, children [SoundClass_0], properties {pitchMax=1, pitchMin=1, volumeMax=0.7, volumeMin=0.7}}
 {
  id=ParamCrossFade_1
  children [Branch_1,Random_4,Random_11]
  properties {
   paramName=FatigueLevel
   crossFadeInput [
    {fadeInDistanceEnd=0, fadeInDistanceStart=0, fadeOutDistanceEnd=0.85, fadeOutDistanceStart=0.82, volume=1.3}
    {fadeInDistanceEnd=0.85, fadeInDistanceStart=0.82, fadeOutDistanceEnd=0.96, fadeOutDistanceStart=0.95, volume=2.5}
    {fadeInDistanceEnd=0.96, fadeInDistanceStart=0.95, fadeOutDistanceEnd=1, fadeOutDistanceStart=1, volume=2.2}
   ]
  }
 }
 {
  id=ParamCrossFade_2
  children [Branch_2,Random_12,Random_8]
  properties {
   paramName=FatigueLevel
   crossFadeInput [
    {fadeInDistanceEnd=0, fadeInDistanceStart=0, fadeOutDistanceEnd=1, fadeOutDistanceStart=0.9, volume=1}
    {fadeInDistanceEnd=0.92, fadeInDistanceStart=0.88, fadeOutDistanceEnd=1, fadeOutDistanceStart=1, volume=1.5}
    {fadeInDistanceEnd=0, fadeInDistanceStart=0, fadeOutDistanceEnd=0, fadeOutDistanceStart=0, volume=1}
   ]
  }
 }
 {id=QualityLevel_0, children [ParamCrossFade_1, ParamCrossFade_1, ParamCrossFade_2, ParamCrossFade_2]}
 {
  id=Random_0
  children [
   WavePlayer_0
   WavePlayer_1
   WavePlayer_2
   WavePlayer_3
   WavePlayer_4
   WavePlayer_5
   WavePlayer_6
   WavePlayer_7
   WavePlayer_8
   WavePlayer_9
   WavePlayer_10
   WavePlayer_11
   WavePlayer_12
   WavePlayer_13
   WavePlayer_14
   WavePlayer_15
   WavePlayer_16
   WavePlayer_17
   WavePlayer_18
   WavePlayer_19
   null
   null
   null
   null
   null
   null
   null
   null
   null
   null
   null
   null
  ]
  properties {weights [1,1,1,1,1,1,1,1,1,1,1,1,1,1,1,1,1,1,1,1,1,1,1,1,1,1,1,1,1,1,1,1]}
 }
 {
  id=Random_11
  children [WavePlayer_65, WavePlayer_67, WavePlayer_69, WavePlayer_73, WavePlayer_76, WavePlayer_77, WavePlayer_78, null, null, null]
  properties {bShouldExcludeFromBranchCulling=true, weights [1,1,1,1,1,1,1,1,1,1]}
 }
 {
  id=Random_12
  children [WavePlayer_89, WavePlayer_94, WavePlayer_95, WavePlayer_96, null, null, null, null, null, null]
  properties {bShouldExcludeFromBranchCulling=true, weights [1,1,1,1,1,1,1,1,1,1]}
 }
 {
  id=Random_13
  children [WavePlayer_97, WavePlayer_98, WavePlayer_99, WavePlayer_100, null, null, null, null]
  properties {bShouldExcludeFromBranchCulling=true, weights [1,1,1,1,1,1,1,1]}
 }
 {
  id=Random_2
  children [
   WavePlayer_48
   WavePlayer_51
   WavePlayer_54
   WavePlayer_62
   WavePlayer_63
   WavePlayer_64
   WavePlayer_66
   WavePlayer_68
   WavePlayer_70
   WavePlayer_71
   WavePlayer_72
   WavePlayer_74
   WavePlayer_75
   null
   null
   null
   null
   null
   null
   null
   null
   null
   null
  ]
  properties {weights [1,1,1,1,1,1,1,1,1,1,1,1,1,1,1,1,1,1,1,1,1,1,1]}
 }
 {
  id=Random_4
  children [
   WavePlayer_44
   WavePlayer_45
   WavePlayer_46
   WavePlayer_47
   WavePlayer_49
   WavePlayer_50
   WavePlayer_52
   WavePlayer_58
   null
   null
   null
   null
   null
  ]
  properties {weights [1,1,1,1,1,1,1,1,1,1,1,1,1]}
 }
 {
  id=Random_5
  children [WavePlayer_90, WavePlayer_91, WavePlayer_92, WavePlayer_93, null, null, null, null]
  properties {bShouldExcludeFromBranchCulling=true, weights [1,1,1,1,1,1,1,1]}
 }
 {
  id=Random_8
  children [WavePlayer_83, WavePlayer_84, WavePlayer_85, WavePlayer_86, WavePlayer_87, WavePlayer_88, null, null, null, null, null, null]
  properties {bShouldExcludeFromBranchCulling=true, weights [1,1,1,1,1,1,1,1,1,1,1,1]}
 }
 {
  id=SoundClass_0
  children [Attenuation_0]
  properties {soundClassOverride=&Fort_Player_Foley_3P_Enemy_Athena}
 }
 {
  id=SoundClass_1
  children [Attenuation_1]
  properties {soundClassOverride=&Fort_Creatures_Vocalization_Tactical}
 }
 {
  id=SoundClass_2
  children [Attenuation_2]
  properties {soundClassOverride=&Fort_Creatures_Vocalization_Tactical}
 }
 {
  id=SoundClass_7
  children [Attenuation_7]
  properties {soundClassOverride=&Fort_Creatures_Vocalization_Tactical}
 }
 {id=WavePlayer_0, properties {soundWaveAssetPtr=&/Grandma_Vox_Walk_Breath_10}}
 {id=WavePlayer_1, properties {soundWaveAssetPtr=&/Grandma_Vox_Walk_Breath_11}}
 {id=WavePlayer_10, properties {soundWaveAssetPtr=&/Grandma_Vox_Walk_Breath_20}}
 {id=WavePlayer_100, properties {soundWaveAssetPtr=&/Grandma_Vox_Walk_Breath_04}}
 {id=WavePlayer_11, properties {soundWaveAssetPtr=&/Grandma_Vox_Walk_Breath_21}}
 {id=WavePlayer_12, properties {soundWaveAssetPtr=&/Grandma_Vox_Walk_Breath_22}}
 {id=WavePlayer_13, properties {soundWaveAssetPtr=&/Grandma_Vox_Walk_Breath_23}}
 {id=WavePlayer_14, properties {soundWaveAssetPtr=&/Grandma_Vox_Walk_Breath_24}}
 {id=WavePlayer_15, properties {soundWaveAssetPtr=&/Grandma_Vox_Walk_Breath_25}}
 {id=WavePlayer_16, properties {soundWaveAssetPtr=&/Grandma_Vox_Walk_Breath_26}}
 {id=WavePlayer_17, properties {soundWaveAssetPtr=&/Grandma_Vox_Walk_Breath_01}}
 {id=WavePlayer_18, properties {soundWaveAssetPtr=&/Grandma_Vox_Walk_Breath_02}}
 {id=WavePlayer_19, properties {soundWaveAssetPtr=&/Grandma_Vox_Walk_Breath_03}}
 {id=WavePlayer_2, properties {soundWaveAssetPtr=&/Grandma_Vox_Walk_Breath_12}}
 {id=WavePlayer_3, properties {soundWaveAssetPtr=&/Grandma_Vox_Walk_Breath_13}}
 {id=WavePlayer_4, properties {soundWaveAssetPtr=&/Grandma_Vox_Walk_Breath_14}}
 {id=WavePlayer_44, properties {soundWaveAssetPtr=&/Grandma_Vocal_Heavy_Breathing_01}}
 {id=WavePlayer_45, properties {soundWaveAssetPtr=&/Grandma_Vocal_Heavy_Breathing_02}}
 {id=WavePlayer_46, properties {soundWaveAssetPtr=&/Grandma_Vocal_Heavy_Breathing_03}}
 {id=WavePlayer_47, properties {soundWaveAssetPtr=&/Grandma_Vocal_Heavy_Breathing_04}}
 {id=WavePlayer_48, properties {soundWaveAssetPtr=&/Grandma_Vox_Swim_Aggro_Breath_01}}
 {id=WavePlayer_49, properties {soundWaveAssetPtr=&/Grandma_Vocal_Heavy_Breathing_05}}
 {id=WavePlayer_5, properties {soundWaveAssetPtr=&/Grandma_Vox_Walk_Breath_15}}
 {id=WavePlayer_50, properties {soundWaveAssetPtr=&/Grandma_Vocal_Heavy_Breathing_06}}
 {id=WavePlayer_51, properties {soundWaveAssetPtr=&/Grandma_Vox_Swim_Aggro_Breath_02}}
 {id=WavePlayer_52, properties {soundWaveAssetPtr=&/Grandma_Vocal_Heavy_Breathing_07}}
 {id=WavePlayer_54, properties {soundWaveAssetPtr=&/Grandma_Vox_Swim_Aggro_Breath_03}}
 {id=WavePlayer_58, properties {soundWaveAssetPtr=&/Grandma_Vocal_Heavy_Breathing_08}}
 {id=WavePlayer_6, properties {soundWaveAssetPtr=&/Grandma_Vox_Walk_Breath_16}}
 {id=WavePlayer_62, properties {soundWaveAssetPtr=&/Grandma_Vox_Swim_Aggro_Breath_07}}
 {id=WavePlayer_63, properties {soundWaveAssetPtr=&/Grandma_Vox_Swim_Aggro_Breath_08}}
 {id=WavePlayer_64, properties {soundWaveAssetPtr=&/Grandma_Vox_Swim_Aggro_Breath_09}}
 {id=WavePlayer_65, properties {soundWaveAssetPtr=&/Grandma_Vocal_Sprint_Tiring_Whines_01}}
 {id=WavePlayer_66, properties {soundWaveAssetPtr=&/Grandma_Vox_Swim_Aggro_Breath_11}}
 {id=WavePlayer_67, properties {soundWaveAssetPtr=&/Grandma_Vocal_Sprint_Tiring_Whines_02}}
 {id=WavePlayer_68, properties {soundWaveAssetPtr=&/Grandma_Vox_Swim_Aggro_Breath_13}}
 {id=WavePlayer_69, properties {soundWaveAssetPtr=&/Grandma_Vocal_Sprint_Tiring_Whines_03}}
 {id=WavePlayer_7, properties {soundWaveAssetPtr=&/Grandma_Vox_Walk_Breath_17}}
 {id=WavePlayer_70, properties {soundWaveAssetPtr=&/Grandma_Vox_Swim_Aggro_Breath_15}}
 {id=WavePlayer_71, properties {soundWaveAssetPtr=&/Grandma_Vox_Swim_Aggro_Breath_16}}
 {id=WavePlayer_72, properties {soundWaveAssetPtr=&/Grandma_Vox_Swim_Aggro_Breath_17}}
 {id=WavePlayer_73, properties {soundWaveAssetPtr=&/Grandma_Vocal_Sprint_Tiring_Whines_04}}
 {id=WavePlayer_74, properties {soundWaveAssetPtr=&/Grandma_Vox_Swim_Aggro_Breath_19}}
 {id=WavePlayer_75, properties {soundWaveAssetPtr=&/Grandma_Vox_Swim_Aggro_Breath_20}}
 {id=WavePlayer_76, properties {soundWaveAssetPtr=&Grandma_Vox_Swim_Aggro_Breath_01}}
 {id=WavePlayer_77, properties {soundWaveAssetPtr=&Grandma_Vox_Swim_Aggro_Breath_02}}
 {id=WavePlayer_78, properties {soundWaveAssetPtr=&Grandma_Vox_Swim_Aggro_Breath_03}}
 {id=WavePlayer_8, properties {soundWaveAssetPtr=&/Grandma_Vox_Walk_Breath_18}}
 {id=WavePlayer_83, properties {soundWaveAssetPtr=&Grandma_Vocal_Sprint_Tiring_Whines_01}}
 {id=WavePlayer_84, properties {soundWaveAssetPtr=&Grandma_Vocal_Sprint_Tiring_Whines_02}}
 {id=WavePlayer_85, properties {soundWaveAssetPtr=&Grandma_Vocal_Sprint_Tiring_Whines_03}}
 {id=WavePlayer_86, properties {soundWaveAssetPtr=&Grandma_Vox_Swim_Aggro_Breath_01}}
 {id=WavePlayer_87, properties {soundWaveAssetPtr=&Grandma_Vox_Swim_Aggro_Breath_02}}
 {id=WavePlayer_88, properties {soundWaveAssetPtr=&Grandma_Vox_Swim_Aggro_Breath_03}}
 {id=WavePlayer_89, properties {soundWaveAssetPtr=&Grandma_Vocal_Heavy_Breathing_01}}
 {id=WavePlayer_9, properties {soundWaveAssetPtr=&/Grandma_Vox_Walk_Breath_19}}
 {id=WavePlayer_90, properties {soundWaveAssetPtr=&Grandma_Vox_Swim_Aggro_Breath_01}}
 {id=WavePlayer_91, properties {soundWaveAssetPtr=&Grandma_Vox_Swim_Aggro_Breath_02}}
 {id=WavePlayer_92, properties {soundWaveAssetPtr=&Grandma_Vox_Swim_Aggro_Breath_03}}
 {id=WavePlayer_93, properties {soundWaveAssetPtr=&/Grandma_Vox_Swim_Aggro_Breath_04}}
 {id=WavePlayer_94, properties {soundWaveAssetPtr=&Grandma_Vocal_Heavy_Breathing_02}}
 {id=WavePlayer_95, properties {soundWaveAssetPtr=&Grandma_Vocal_Heavy_Breathing_03}}
 {id=WavePlayer_96, properties {soundWaveAssetPtr=&Grandma_Vocal_Heavy_Breathing_04}}
 {id=WavePlayer_97, properties {soundWaveAssetPtr=&Grandma_Vox_Walk_Breath_01}}
 {id=WavePlayer_98, properties {soundWaveAssetPtr=&Grandma_Vox_Walk_Breath_02}}
 {id=WavePlayer_99, properties {soundWaveAssetPtr=&Grandma_Vox_Walk_Breath_03}}
]
properties {
 maxDistance=9000
 pitchMultiplier=0.8
 volumeMultiplier=1
 concurrencySet [
  &Athena_Footstep_Local_Conc
  &Athena_Footstep_Global_Conc
  /Game/Sounds/Concurrency_Rules/LimitToOnePreventNewQuick
 ]
}
