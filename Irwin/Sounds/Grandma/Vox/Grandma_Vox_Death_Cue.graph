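# Sound cue for Grandma's death vocalization, with predator vocalization attenuation and a tactical creature vocalization sound class.
# A quality-level switch sends its first two levels to a random pick among eleven death recordings.
# The other two levels use a random pick among three of those recordings, excluded from branch culling and pitch-modulated to add variety.
# Concurrency reuses the Athena local and global footstep sets, and volume is boosted.
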SoundCue /Irwin/Sounds/Grandma/Vox/Grandma_Vox_Death_Cue
root=QualityLevel_1
nodes [
 {id=Modulator_0, children [Random_0], properties {pitchMax=1.1, pitchMin=0.9}}
 {id=QualityLevel_1, children [Random_2,Random_2,Modulator_0,Modulator_0]}
 {id=Random_0, children [WavePlayer_0,WavePlayer_1,WavePlayer_2], properties {bShouldExcludeFromBranchCulling=true, weights [1,1,1]}}
 {
  id=Random_2
  children [
   WavePlayer_19
   WavePlayer_20
   WavePlayer_21
   WavePlayer_22
   WavePlayer_23
   WavePlayer_24
   WavePlayer_25
   WavePlayer_26
   WavePlayer_27
   WavePlayer_29
   WavePlayer_30
  ]
  properties {weights [1,1,1,1,1,1,1,1,1,1,1]}
 }
 {id=WavePlayer_0, properties {soundWaveAssetPtr=&/Grandma_Vox_Death_01}}
 {id=WavePlayer_1, properties {soundWaveAssetPtr=&/Grandma_Vox_Death_02}}
 {id=WavePlayer_19, properties {soundWaveAssetPtr=&Grandma_Vox_Death_01}}
 {id=WavePlayer_2, properties {soundWaveAssetPtr=&/Grandma_Vox_Death_08}}
 {id=WavePlayer_20, properties {soundWaveAssetPtr=&Grandma_Vox_Death_02}}
 {id=WavePlayer_21, properties {soundWaveAssetPtr=&/Grandma_Vox_Death_03}}
 {id=WavePlayer_22, properties {soundWaveAssetPtr=&/Grandma_Vox_Death_04}}
 {id=WavePlayer_23, properties {soundWaveAssetPtr=&/Grandma_Vox_Death_05}}
 {id=WavePlayer_24, properties {soundWaveAssetPtr=&/Grandma_Vox_Death_06}}
 {id=WavePlayer_25, properties {soundWaveAssetPtr=&Grandma_Vox_Death_08}}
 {id=WavePlayer_26, properties {soundWaveAssetPtr=&/Grandma_Vox_Death_09}}
 {id=WavePlayer_27, properties {soundWaveAssetPtr=&/Grandma_Vox_Death_10}}
 {id=WavePlayer_29, properties {soundWaveAssetPtr=&/Grandma_Vox_Death_12}}
 {id=WavePlayer_30, properties {soundWaveAssetPtr=&/Grandma_Vox_Death_13}}
]
properties {
 attenuationSettings=/Irwin/Sounds/Settings/Predator_Vocalization_Att
 maxDistance=8300
 soundClassObject=&Fort_Creatures_Vocalization_Tactical
 volumeMultiplier=3
 concurrencySet [&Athena_Footstep_Local_Conc, &Athena_Footstep_Global_Conc]
}
