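# Sound cue for Grandma's eating vocalization, played as spatialized creature vocalization with predator attenuation.
# A mixer layers a random swim aggro breath over a quieter random eat vocal, which is slightly delayed and pitched down.
# An outer modulator adds small pitch variation, and a quality switch feeds the mix to only the first two of four levels.
# Concurrency reuses the Athena footstep local and global settings.
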SoundCue /Irwin/Sounds/Grandma/Vox/Grandma_Vox_Eat_Cue
root=Modulator_0
nodes [
 {id=Delay_1, children [Random_2], properties {delayMax=0.1, delayMin=0.1}}
 {id=Mixer_0, children [Random_0,Modulator_1], properties {inputVolume [1.3,0.7]}}
 {id=Modulator_0, children [QualityLevel_0], properties {pitchMax=1.02, pitchMin=0.98}}
 {id=Modulator_1, children [Delay_1], properties {pitchMax=0.9, pitchMin=0.8, volumeMax=1.2, volumeMin=1}}
 {id=QualityLevel_0, children [Mixer_0,Mixer_0,null,null]}
 {
  id=Random_0
  children [WavePlayer_8, WavePlayer_9, WavePlayer_10, WavePlayer_3, WavePlayer_5, WavePlayer_6, WavePlayer_7]
  properties {weights [1,1,1,1,1,1,1]}
 }
 {id=Random_2, children [WavePlayer_12, WavePlayer_13, WavePlayer_14, WavePlayer_15, WavePlayer_16], properties {weights [1,1,1,1,1]}}
 {id=WavePlayer_10, properties {soundWaveAssetPtr=&/Grandma_Vox_Swim_Aggro_Breath_03}}
 {id=WavePlayer_12, properties {soundWaveAssetPtr=&/Grandma_Vocal_Eat_01}}
 {id=WavePlayer_13, properties {soundWaveAssetPtr=&/Grandma_Vocal_Eat_02}}
 {id=WavePlayer_14, properties {soundWaveAssetPtr=&/Grandma_Vocal_Eat_03}}
 {id=WavePlayer_15, properties {soundWaveAssetPtr=&/Grandma_Vocal_Eat_04}}
 {id=WavePlayer_16, properties {soundWaveAssetPtr=&/Grandma_Vocal_Eat_05}}
 {id=WavePlayer_3, properties {soundWaveAssetPtr=&/Grandma_Vox_Swim_Aggro_Breath_13}}
 {id=WavePlayer_5, properties {soundWaveAssetPtr=&/Grandma_Vox_Swim_Aggro_Breath_11}}
 {id=WavePlayer_6, properties {soundWaveAssetPtr=&/Grandma_Vox_Swim_Aggro_Breath_09}}
 {id=WavePlayer_7, properties {soundWaveAssetPtr=&/Grandma_Vox_Swim_Aggro_Breath_07}}
 {id=WavePlayer_8, properties {soundWaveAssetPtr=&/Grandma_Vox_Swim_Aggro_Breath_01}}
 {id=WavePlayer_9, properties {soundWaveAssetPtr=&/Grandma_Vox_Swim_Aggro_Breath_02}}
]
properties {
 attenuationSettings=/Irwin/Sounds/Settings/Predator_Vocalization_Att
 maxDistance=8300
 soundClassObject=/Game/Sounds/FortSoundClassesAndModes/Fort_Creatures_Vocalization_Normal
 volumeMultiplier=1.2
 concurrencySet [&Athena_Footstep_Local_Conc, &Athena_Footstep_Global_Conc]
}
