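# Sound cue for Grandma's howl vocalization, a creature sound layering close and distant recordings.
# A quality switch routes its first two levels to a mixer with the full variation pools and its last two to one with only two variations per layer.
# Each mixer blends randomized, pitched-down close howls with quieter distant howls, each layer using its own attenuation.
# A concurrency override caps playback at two instances, stopping the farthest first.
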SoundCue /Irwin/Sounds/Grandma/Vox/Grandma_Vox_Howl_Cue
root=QualityLevel_0
nodes [
 {id=Attenuation_0, children [Random_1], properties {attenuationSettings=/Irwin/Sounds/Settings/Predator_Vocalization_Howl_Dist_Att}}
 {id=Attenuation_1, children [Modulator_0], properties {attenuationSettings=/Irwin/Sounds/Settings/Predator_Vocalization_Howl_Close_Att}}
 {id=Attenuation_2, children [Random_3], properties {attenuationSettings=&Predator_Vocalization_Howl_Dist_Att}}
 {id=Attenuation_3, children [Modulator_1], properties {attenuationSettings=&Predator_Vocalization_Howl_Close_Att}}
 {id=Mixer_0, children [Attenuation_1, Attenuation_0], properties {inputVolume [1,0.3]}}
 {id=Mixer_1, children [Attenuation_3, Attenuation_2], properties {inputVolume [1,0.3]}}
 {id=Modulator_0, children [Random_0], properties {pitchMax=0.9, pitchMin=0.9, volumeMax=1, volumeMin=1}}
 {id=Modulator_1, children [Random_2], properties {pitchMax=0.9, pitchMin=0.9, volumeMax=1, volumeMin=1}}
 {id=QualityLevel_0, children [Mixer_0,Mixer_0,Mixer_1,Mixer_1]}
 {
  id=Random_0
  children [WavePlayer_0,WavePlayer_1,WavePlayer_2,WavePlayer_3,WavePlayer_4,WavePlayer_5,WavePlayer_6]
  properties {bShouldExcludeFromBranchCulling=true, weights [1,1,1,1,1,1,1]}
 }
 {
  id=Random_1
  children [WavePlayer_7, WavePlayer_8, WavePlayer_9, WavePlayer_10, WavePlayer_11, WavePlayer_12]
  properties {bShouldExcludeFromBranchCulling=true, weights [1,1,1,1,1,1]}
 }
 {id=Random_2, children [WavePlayer_13, WavePlayer_14], properties {bShouldExcludeFromBranchCulling=true, weights [1,1]}}
 {id=Random_3, children [WavePlayer_20, WavePlayer_21], properties {bShouldExcludeFromBranchCulling=true, weights [1,1]}}
 {id=WavePlayer_0, properties {soundWaveAssetPtr=&/Grandma_Vox_Howl_01}}
 {id=WavePlayer_1, properties {soundWaveAssetPtr=&/Grandma_Vox_Howl_02}}
 {id=WavePlayer_10, properties {soundWaveAssetPtr=&/Grandma_Vox_Howl_Distant_04}}
 {id=WavePlayer_11, properties {soundWaveAssetPtr=&/Grandma_Vox_Howl_Distant_05}}
 {id=WavePlayer_12, properties {soundWaveAssetPtr=&/Grandma_Vox_Howl_Distant_06}}
 {id=WavePlayer_13, properties {soundWaveAssetPtr=&Grandma_Vox_Howl_01}}
 {id=WavePlayer_14, properties {soundWaveAssetPtr=&Grandma_Vox_Howl_02}}
 {id=WavePlayer_2, properties {soundWaveAssetPtr=&/Grandma_Vox_Howl_03}}
 {id=WavePlayer_20, properties {soundWaveAssetPtr=&/Grandma_Vox_Howl_Distant_01}}
 {id=WavePlayer_21, properties {soundWaveAssetPtr=&/Grandma_Vox_Howl_Distant_02}}
 {id=WavePlayer_3, properties {soundWaveAssetPtr=&/Grandma_Vox_Howl_04}}
 {id=WavePlayer_4, properties {soundWaveAssetPtr=&/Grandma_Vox_Howl_05}}
 {id=WavePlayer_5, properties {soundWaveAssetPtr=&/Grandma_Vox_Howl_06}}
 {id=WavePlayer_6, properties {soundWaveAssetPtr=&/Grandma_Vox_Howl_07}}
 {id=WavePlayer_7, properties {soundWaveAssetPtr=&Grandma_Vox_Howl_Distant_01}}
 {id=WavePlayer_8, properties {soundWaveAssetPtr=&Grandma_Vox_Howl_Distant_02}}
 {id=WavePlayer_9, properties {soundWaveAssetPtr=&/Grandma_Vox_Howl_Distant_03}}
]
properties {
 bOverrideConcurrency=true
 maxDistance=18000
 soundClassObject=&Fort_Creatures_Vocalization_Tactical
 volumeMultiplier=2
 concurrencyOverrides {maxCount=2, resolutionRule=StopFarthestThenPreventNew}
 concurrencySet [&Athena_Footstep_Local_Conc, &Athena_Footstep_Global_Conc]
}
