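# Sound cue for Grandma's idle-break vocalization, using predator vocalization attenuation and the creature vocalization sound class.
# A quality-level switch plays a random pick from four waves on the first two levels.
# The other two levels pick from three wave players, the second wave appearing twice, through slight random pitch modulation.
# Concurrency is limited by Athena footstep local and global sets rather than vocal-specific ones.
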SoundCue /Irwin/Sounds/Grandma/Vox/Grandma_Vox_Idle_Break_Cue
root=QualityLevel_0
nodes [
 {id=Modulator_0, children [Random_1], properties {pitchMax=1.02, pitchMin=0.98}}
 {id=QualityLevel_0, children [Random_0,Random_0,Modulator_0,Modulator_0]}
 {id=Random_0, children [WavePlayer_0,WavePlayer_1,WavePlayer_2,WavePlayer_3], properties {weights [1,1,1,1]}}
 {id=Random_1, children [WavePlayer_4,WavePlayer_5,WavePlayer_6], properties {bShouldExcludeFromBranchCulling=true, weights [1,1,1]}}
 {id=WavePlayer_0, properties {soundWaveAssetPtr=&/Grandma_Vox_Idle_Break_01}}
 {id=WavePlayer_1, properties {soundWaveAssetPtr=&/Grandma_Vox_Idle_Break_02}}
 {id=WavePlayer_2, properties {soundWaveAssetPtr=&/Grandma_Vox_Idle_Break_03}}
 {id=WavePlayer_3, properties {soundWaveAssetPtr=&/Grandma_Vox_Idle_Break_04}}
 {id=WavePlayer_4, properties {soundWaveAssetPtr=&Grandma_Vox_Idle_Break_01}}
 {id=WavePlayer_5, properties {soundWaveAssetPtr=&Grandma_Vox_Idle_Break_02}}
 {id=WavePlayer_6, properties {soundWaveAssetPtr=&Grandma_Vox_Idle_Break_02}}
]
properties {
 attenuationSettings=/Irwin/Sounds/Settings/Predator_Vocalization_Att
 maxDistance=8300
 soundClassObject=/Game/Sounds/FortSoundClassesAndModes/Fort_Creatures_Vocalization_Normal
 volumeMultiplier=1
 concurrencySet [&Athena_Footstep_Local_Conc, &Athena_Footstep_Global_Conc]
}
