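# Sound Cue for Grandma's short idle vocalizations, using predator vocalization attenuation and the creature vocalization sound class.
# A quality-level node sends the first two levels to an equal-weight random pick of thirteen waves.
# The other two levels use a random pick of only four waves through a pitch and volume modulator, adding variation to the smaller set.
# Concurrency reuses the Athena local and global footstep settings.
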SoundCue /Irwin/Sounds/Grandma/Vox/Grandma_Vox_Idle_Short_Cue
root=QualityLevel_0
nodes [
 {id=Modulator_0, children [Random_1], properties {pitchMax=1.1, pitchMin=0.9, volumeMin=0.6}}
 {id=QualityLevel_0, children [Random_2,Random_2,Modulator_0,Modulator_0]}
 {
  id=Random_1
  children [WavePlayer_13, WavePlayer_14, WavePlayer_15, WavePlayer_16]
  properties {bShouldExcludeFromBranchCulling=true, weights [1,1,1,1]}
 }
 {
  id=Random_2
  children [
   WavePlayer_0
   WavePlayer_1
   WavePlayer_2
   WavePlayer_3
   WavePlayer_4
   WavePlayer_5
   WavePlayer_6
   WavePlayer_7
   WavePlayer_8
   WavePlayer_9
   WavePlayer_10
   WavePlayer_11
   WavePlayer_12
  ]
  properties {weights [1,1,1,1,1,1,1,1,1,1,1,1,1]}
 }
 {id=WavePlayer_0, properties {soundWaveAssetPtr=&/Grandma_Vox_Idle_Short_09}}
 {id=WavePlayer_1, properties {soundWaveAssetPtr=&/Grandma_Vox_Idle_Short_10}}
 {id=WavePlayer_10, properties {soundWaveAssetPtr=&/Grandma_Vox_Idle_Short_06}}
 {id=WavePlayer_11, properties {soundWaveAssetPtr=&/Grandma_Vox_Idle_Short_07}}
 {id=WavePlayer_12, properties {soundWaveAssetPtr=&/Grandma_Vox_Idle_Short_08}}
 {id=WavePlayer_13, properties {soundWaveAssetPtr=&/Grandma_Vox_Idle_Short_01}}
 {id=WavePlayer_14, properties {soundWaveAssetPtr=&/Grandma_Vox_Idle_Short_02}}
 {id=WavePlayer_15, properties {soundWaveAssetPtr=&/Grandma_Vox_Idle_Short_03}}
 {id=WavePlayer_16, properties {soundWaveAssetPtr=&/Grandma_Vox_Idle_Short_04}}
 {id=WavePlayer_2, properties {soundWaveAssetPtr=&/Grandma_Vox_Idle_Short_11}}
 {id=WavePlayer_3, properties {soundWaveAssetPtr=&/Grandma_Vox_Idle_Short_12}}
 {id=WavePlayer_4, properties {soundWaveAssetPtr=&/Grandma_Vox_Idle_Short_13}}
 {id=WavePlayer_5, properties {soundWaveAssetPtr=&Grandma_Vox_Idle_Short_01}}
 {id=WavePlayer_6, properties {soundWaveAssetPtr=&Grandma_Vox_Idle_Short_02}}
 {id=WavePlayer_7, properties {soundWaveAssetPtr=&Grandma_Vox_Idle_Short_03}}
 {id=WavePlayer_8, properties {soundWaveAssetPtr=&Grandma_Vox_Idle_Short_04}}
 {id=WavePlayer_9, properties {soundWaveAssetPtr=&/Grandma_Vox_Idle_Short_05}}
]
properties {
 attenuationSettings=/Irwin/Sounds/Settings/Predator_Vocalization_Att
 maxDistance=8300
 soundClassObject=/Game/Sounds/FortSoundClassesAndModes/Fort_Creatures_Vocalization_Normal
 concurrencySet [&Athena_Footstep_Local_Conc, &Athena_Footstep_Global_Conc]
}
