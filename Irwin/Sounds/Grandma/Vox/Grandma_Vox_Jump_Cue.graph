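# Sound cue for Grandma's jump vocalization, varying by who hears it through the IsLocal and IsRiddenByEnemy parameters.
# The local path is boosted and randomly picks a jump vocal or a quieter heavy breath.
# Other paths share randomized, modulated jump vocals, louder with enemy foley class and riding attenuation when enemy-ridden, otherwise very quiet with creature vocalization settings.
# Footstep and limit-to-one concurrency rules restrict overlapping plays.
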SoundCue /Irwin/Sounds/Grandma/Vox/Grandma_Vox_Jump_Cue
root=Branch_4
nodes [
 {id=Attenuation_5, children [Modulator_3], properties {attenuationSettings=/Game/Sounds/Attenuation/Fort_Riding_Enemy_VO_Att}}
 {id=Attenuation_6, children [Modulator_3], properties {attenuationSettings=/Irwin/Sounds/Settings/Predator_Vocalization_Att}}
 {id=Attenuation_7, children [Random_4], properties {attenuationSettings=&Predator_Vocalization_Att}}
 {id=Branch_3, children [Modulator_5,Modulator_4,Modulator_4], properties {boolParameterName=IsRiddenByEnemy}}
 {id=Branch_4, children [Modulator_0,Branch_3,Branch_3], properties {boolParameterName=IsLocal}}
 {id=Mixer_0, children [Random_3], properties {inputVolume [1]}}
 {id=Mixer_1, children [Random_0], properties {inputVolume [0.5]}}
 {id=Modulator_0, children [SoundClass_7], properties {pitchMax=1, pitchMin=1, volumeMax=2, volumeMin=2}}
 {id=Modulator_3, children [Random_2], properties {volumeMax=1, volumeMin=0.5}}
 {id=Modulator_4, children [SoundClass_6], properties {pitchMax=1, pitchMin=1, volumeMax=0.1, volumeMin=0.1}}
 {id=Modulator_5, children [SoundClass_5], properties {pitchMax=1, pitchMin=1, volumeMax=0.7, volumeMin=0.7}}
 {
  id=Random_0
  children [WavePlayer_42, WavePlayer_43, WavePlayer_44, WavePlayer_45, WavePlayer_46, WavePlayer_47, WavePlayer_48, WavePlayer_49]
  properties {bShouldExcludeFromBranchCulling=true, weights [1,1,1,1,1,1,1,1]}
 }
 {id=Random_2, children [WavePlayer_6,WavePlayer_7,WavePlayer_9], properties {bShouldExcludeFromBranchCulling=true, weights [1,1,1]}}
 {id=Random_3, children [WavePlayer_3, WavePlayer_8, WavePlayer_10], properties {bShouldExcludeFromBranchCulling=true, weights [1,1,1]}}
 {id=Random_4, children [Mixer_0,Mixer_1], properties {weights [1,1]}}
 {
  id=SoundClass_5
  children [Attenuation_5]
  properties {soundClassOverride=&Fort_Player_Foley_3P_Enemy_Athena}
 }
 {
  id=SoundClass_6
  children [Attenuation_6]
  properties {soundClassOverride=&Fort_Creatures_Vocalization_Tactical}
 }
 {
  id=SoundClass_7
  children [Attenuation_7]
  properties {soundClassOverride=&Fort_Creatures_Vocalization_Tactical}
 }
 {id=WavePlayer_10, properties {soundWaveAssetPtr=&/Grandma_Vox_Jump_03}}
 {id=WavePlayer_3, properties {soundWaveAssetPtr=&/Grandma_Vox_Jump_04}}
 {id=WavePlayer_42, properties {soundWaveAssetPtr=&/Grandma_Vocal_Heavy_Breathing_01}}
 {id=WavePlayer_43, properties {soundWaveAssetPtr=&/Grandma_Vocal_Heavy_Breathing_02}}
 {id=WavePlayer_44, properties {soundWaveAssetPtr=&/Grandma_Vocal_Heavy_Breathing_03}}
 {id=WavePlayer_45, properties {soundWaveAssetPtr=&/Grandma_Vocal_Heavy_Breathing_04}}
 {id=WavePlayer_46, properties {soundWaveAssetPtr=&/Grandma_Vocal_Heavy_Breathing_05}}
 {id=WavePlayer_47, properties {soundWaveAssetPtr=&/Grandma_Vocal_Heavy_Breathing_06}}
 {id=WavePlayer_48, properties {soundWaveAssetPtr=&/Grandma_Vocal_Heavy_Breathing_07}}
 {id=WavePlayer_49, properties {soundWaveAssetPtr=&/Grandma_Vocal_Heavy_Breathing_08}}
 {id=WavePlayer_6, properties {soundWaveAssetPtr=&Grandma_Vox_Jump_04}}
 {id=WavePlayer_7, properties {soundWaveAssetPtr=&/Grandma_Vox_Jump_01}}
 {id=WavePlayer_8, properties {soundWaveAssetPtr=&Grandma_Vox_Jump_01}}
 {id=WavePlayer_9, properties {soundWaveAssetPtr=&Grandma_Vox_Jump_03}}
]
properties {
 maxDistance=9000
 volumeMultiplier=1.25
 concurrencySet [
  &Athena_Footstep_Local_Conc
  &Athena_Footstep_Global_Conc
  /Game/Sounds/Concurrency_Rules/LimitToOnePreventNew
 ]
}
